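# Animation Blueprint for the female medium MechPilotSharkVelocity body that copies the attached parent's pose and corrects deformation around hips and thigh discs.
# A deform Control Rig receives fixed Glute and Groin transforms, then half-strength constraints tie disc bases to thigh twist bones.
# Additive bone translations shift each disc, weighted by thigh rotation relative to pelvis, remapped each update.
# A second left-disc offset node and the Disc variable are unconnected.
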AnimBlueprint /Game/Characters/Player/Female/Medium/Bodies/F_MED_MechPilotSharkVelocity/Meshes/F_MED_MechPilotSharkVelocity_AnimBP
target_skeleton=&/F_MED_MechPilotSharkVelocity_Skeleton
graphs [
 AnimGraph {nodes [
   "Transform (Modify) Bone - Bone: dyn_thigh_disc_l"(Alpha=Get thigh_rot_l_0() -> (thigh_rot_l_0:double), Translation="0,0,-2", Rotation="0,0,0", Scale="1,1,1") {
    rotationMode=BMM_Additive
    rotationSpace=BCS_BoneSpace
    scaleMode=BMM_Additive
    scaleSpace=BCS_BoneSpace
    translationMode=BMM_Additive
    translationSpace=BCS_BoneSpace
    alphaBoolBlend {blendOption=Linear}
    boneToModify {boneName=dyn_thigh_disc_l}
    translation {z=-2}
   }
   n1 = Control Rig!(
    Source=Copy Pose From Mesh() {bUseAttachedParent=true}
    Alpha
    ModifyTransforms_1_ItemToModify_0_Transform:Transform=Get Glute() -> (Glute:Transform)
    ModifyTransforms_ItemToModify_0_Transform:Transform=Get Groin() -> (Groin:Transform)
   ) -> (Pose) {
    alphaBoolBlend {blendOption=Linear}
    controlRigAssetReference {blueprintRigClass=&/F_MED_MechPilotSharkVelocity_DeformRig.F_MED_MechPilotSharkVelocity_DeformRig_C}
   }
   n2 = "Constraint - dyn_thigh_disc_base_l (P)"(ComponentPose=Local To Component(LocalPose=@n1), Alpha=0.5, ConstraintWeights_0=1) {
    alphaBoolBlend {blendOption=Linear}
    boneToModify {boneName=dyn_thigh_disc_base_l}
    constraintSetup [{targetBone {boneName=thigh_twist_01_l}}]
    constraintWeights [1]
   }
   n3 = "Constraint - dyn_thigh_disc_base_r (P)"(ComponentPose=@n2, Alpha=0.5, ConstraintWeights_0=1) {
    alphaBoolBlend {blendOption=Linear}
    boneToModify {boneName=dyn_thigh_disc_base_r}
    constraintSetup [{targetBone {boneName=thigh_twist_01_r}}]
    constraintWeights [1]
   }
   n4 = "Transform (Modify) Bone - Bone: dyn_thigh_disc_l"(ComponentPose=@n3, Alpha=Get thigh_rot_l(), Translation="-0.509,1.332,-1.742", Rotation="0,0,0", Scale="1,1,1") {
    rotationMode=BMM_Additive
    rotationSpace=BCS_BoneSpace
    scaleMode=BMM_Additive
    scaleSpace=BCS_BoneSpace
    translationMode=BMM_Additive
    translationSpace=BCS_BoneSpace
    alphaBoolBlend {blendOption=Linear}
    boneToModify {boneName=dyn_thigh_disc_l}
    translation=<-0.509 1.332 -1.742>
   }
   n5 = "Transform (Modify) Bone - Bone: dyn_thigh_disc_r"(ComponentPose=@n4, Alpha=Get thigh_rot_r(), Translation="0,0,1", Rotation="0,0,0", Scale="1,1,1") {
    rotationMode=BMM_Additive
    rotationSpace=BCS_BoneSpace
    scaleMode=BMM_Additive
    scaleSpace=BCS_BoneSpace
    translationMode=BMM_Additive
    translationSpace=BCS_BoneSpace
    alphaBoolBlend {blendOption=Linear}
    boneToModify {boneName=dyn_thigh_disc_r}
    translation {z=1}
   }
   Output Pose(Result=Component To Local(ComponentPose=@n5)) {layerGroup=None}
   Get Disc() -> (Disc:Transform)
  ]}
 EventGraph {nodes [
   Event BlueprintUpdateAnimation()
   > n1 = Sequence!(execute) -> (then_0, then_1, then_2:Exec, then_3:Exec)
   n2 = GetOwnerSkeletalMeshComponent()
   n3 = BreakRotator(InRot=BreakTransform(InTransform=GetDeltaTransformFromRefPose(self=@n2, BoneName=thigh_l, BaseName=pelvis)).Rotation)
   > Set thigh_rot_l(thigh_rot_l=MapRangeClamped(Value=@n3.Yaw, InRangeA=-20, InRangeB=-90))
   > Set thigh_rot_l_0(thigh_rot_l_0:double=MapRangeClamped(Value=@n3, InRangeA=0, InRangeB=-45)) -> (Output_Get:double)
   n4 = BreakTransform(InTransform=GetDeltaTransformFromRefPose(self=@n2, BoneName=thigh_r, BaseName=pelvis))
   Set thigh_rot_r(execute=@n1.then_1, thigh_rot_r=MapRangeClamped(Value=BreakRotator(InRot=@n4.Rotation).Yaw, InRangeA=0, InRangeB=-105))
  ]}
]
variables [
 Disc {type=Transform, default {location=<0 -1 0>, rotation=<rot 10 -45 20>, scale=<1 1 1>}}
 Glute {type=Transform, default {location=<-8 -1 0>, rotation=<rot 0 0 0>, scale=<0.5 1 1>}}
 Groin {type=Transform, default {location=<-3 12 0>, rotation=<rot 0 0 0>, scale=<0.5 1 1>}}
 thigh_rot_l {type=number}
 thigh_rot_l_0 {type=number}
 thigh_rot_r {type=number}
]
